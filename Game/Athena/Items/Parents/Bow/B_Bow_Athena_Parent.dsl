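; Parent Blueprint for Athena charge-up bows, handling draw audio, charge FX and an aim trajectory preview.
; On charge start the locally controlled player spawns a trajectory actor, and Tick raises velocity with charge up to a cap, redraws the predicted spline and checks its hit target.
; Max charge triggers sound and muzzle FX, plus local shake, rumble and indicator pulse.
; Fire sounds vary by charge and viewer, while firing, release and unequip clean up the preview and audio.
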(event Custom|IncreaseDrawStrength
  (bind _row_maxprojectilespeed (Variables|ParentBow|GetRow_MaxProjectileSpeed))
  (bind _returnvalue (ScalableFloat|GetValueatLevel _row_maxprojectilespeed 1.0))
  ;[1 1 1] Update Trajectory Spline
  (if (<= (+ (Variables|ParentBow|GetProjectileVelocity) 1.0) _returnvalue)
    (Variables|ParentBow|SetProjectileVelocity (CallFunction|CalculateVelocity))
    (else
      (Variables|ParentBow|SetProjectileVelocity _returnvalue)))
  (bind _calculatesplinehitresult (CallFunction|CalculateTrajectorySpline))
  (Class|BPProjectileTrajectoryBowGeneric|CheckValidTarget (Variables|ParentBow|GetBowTrajectory) _calculatesplinehitresult))

(event Custom|DeactivateFullyChargedFX
  (Components|Activation|Deactivate (Variables|Default|GetFX_FullyCharged)))

(event EventOnStartCharge
  ;[1 1 1] EVENT -> StartCharge
  (|CallParentFunction)
  (bind _asfort_player_pawn (Utilities|Casting|CastToFortPlayerPawn (Game|GetInstigator)))
  (bind _output_get (Variables|ParentBow|SetPlayerPawn _asfort_player_pawn))
  (CallFunction|StartDrawAudio)
  (if (Pawn|IsLocallyControlled _output_get)
    (CallFunction|SetupTrajectoryProjectile)
    (CallFunction|SetDummyVariables)))

(event EventOnChargeUp
  ;[1 1 1] EVENT -> On Charge Up
  (|CallParentFunction)
  (Utilities|FlowControl|DoOnce false
    (:Completed
      (Variables|ParentBow|SetCharging true)
      (Variables|ParentBow|SetReachedMinCharge false) ; Can play charged weapon fire sound
      (Variables|ParentBow|SetReachedMaxCharge false))))

(event Custom|ResetCharge
  ;[1 1 1] EVENT -> On Charge Up
  (Variables|ParentBow|SetProjectileVelocity 0.0)
  (Utilities|FlowControl|DoOnce false
    (:Completed
      (Variables|ParentBow|SetCharging true)
      (Variables|ParentBow|SetReachedMinCharge false) ; Can play charged weapon fire sound
      (Variables|ParentBow|SetReachedMaxCharge false))))

(event EventOnLocalClientReachedMaxCharge
  (bind _returnvalue (Game|GetInstigator))
  (bind _returnvalue_2 (Fortnite|GetFortPlayerControllerfromActor _returnvalue))
  ;[1 1 1] EVENT -> Min/Max Charge Reached
  (CallFunction|SetMaxChargeReachedandPlayAudio)
  (CallFunction|SpawnFullyChargedFX)
  (Game|Feedback|ClientStartCameraShake :self _returnvalue_2 :Shake "/Game/Effects/Fort_Effects/Camera_Effects/BP_Camera_Shake_Pulse_Zoom_Small.BP_Camera_Shake_Pulse_Zoom_Small_C" :Scale 1.0 :PlaySpace "CameraLocal")
  (Game|Feedback|ClientPlayForceFeedback _returnvalue_2 "/Game/Weapons/ForceFeedback/FF_ShoppingCartBump.FF_ShoppingCartBump" "None" false true false)
  (bind _bowtrajectory (Variables|ParentBow|GetBowTrajectory)
    (:then
      (Class|BPProjectileTrajectoryBowGeneric|Pulse _bowtrajectory))
    (:else)))

(event EventOnRemoteClientReachedMaxCharge
  ;[1 1 1] EVENT -> Min/Max Charge Reached
  (CallFunction|SetMaxChargeReachedandPlayAudio)
  (CallFunction|SpawnFullyChargedFX))

(event EventOnLocalClientReachedMinCharge
  ;[1 1 1] EVENT -> Min/Max Charge Reached
  (Variables|ParentBow|SetReachedMinCharge true) ; Can play charged weapon fire sound
)

(event EventOnRemoteClientReachedMinCharge
  ;[1 1 1] EVENT -> Min/Max Charge Reached
  (Variables|ParentBow|SetReachedMinCharge true) ; Can play charged weapon fire sound
)

(event Custom|EndOfCharge
  ;[1 1 1] End of Charge (Max Charge Reached)
  ;;[1 0 0 0.2] Investigation into this issue is ongoing by Daniel Broder at the moment.  There are some cases this may not be called on some remote clients; will this cause audio to erroneously get stuck on?
  (CallFunction|StopDrawAudio)
  (Variables|ParentBow|SetProjectileVelocity 0.0)
  (Utilities|Time|ClearandInvalidateTimerbyHandle (Variables|ParentBow|GetTimer_UpdateTrajectorySpline))
  (Actor|DestroyActor (Variables|ParentBow|GetBowTrajectory)))

(event EventOnChargeDown
  ;[1 1 1] EVENT -> Charge Down (mouse released)
  (CallFunction|EndOfCharge)
  (Utilities|FlowControl|Delay 0.034)
  (Variables|ParentBow|SetCharging false))

(event EventOnPlayWeaponFireFX
  (params (bPersistentFire :type Boolean) (bSecondaryFire :type Boolean))
  ;[1 1 1] EVENT -> Weapon Fire FX
  (|CallParentFunction bPersistentFire false)
  (Utilities|FlowControl|Sequence
    (:then_0
      (CallFunction|EndOfCharge))
    (:then_1
      (if (Variables|ParentBow|GetReachedMaxCharge)
        ;;[1 0 0 0.2] Need to review this to determine whether any changes are required.  It should be OK though since this FX is just for firing, and clients that don't do this should be of low significance AND not being observed directly as the target of any camera.
        ;; Investigation by Daniel Broder is ongoing.  This comment will be removed when done.
        (Effects|Components|ParticleSystem|SpawnAltFXAttached "GenericAnimNotify" (Variables|ParentBow|GetNS_AdditionalFire) (Variables|ParentBow|GetFX_AdditionalFire) (Variables|Visual|GetWeaponMesh) "Muzzle")))
    (:then_2
      ;[1 0 0 0.2] Need to review this to determine whether any fixes are required.  This is not guaranteed to be called on all clients.  (Investigation in progress, please see Daniel Broder about any issues here.)
      (CallFunction|DeactivateFullyChargedFX))))

(event FortWeapon|EventOnEquippedAndAttached
  ;[1 1 1] EVENT -> Equip: Added 'IsEquipped' bool
  (Variables|ParentBow|SetIsBowEquipped true))

(event FortWeapon|EventOnLocalUnEquip
  ;[1 1 1] EVENT -> Unequip - to fix a bug if you were charging, then switched weapons before firing
  (Variables|ParentBow|SetIsBowEquipped false)
  (CallFunction|ResetCharge)
  (Variables|ParentBow|SetCharging false)
  (CallFunction|EndOfCharge)
  (Networking|SwitchHasAuthority
    (:Authority
      (CallFunction|DeactivateFullyChargedFX))))

(event EventOnEndCharge
  ;[1 1 1] EVENT -> Unequip - to fix a bug if you were charging, then switched weapons before firing
  (CallFunction|ResetCharge)
  (Variables|ParentBow|SetCharging false)
  (CallFunction|EndOfCharge)
  (Networking|SwitchHasAuthority
    (:Authority
      (CallFunction|DeactivateFullyChargedFX))))

(event EventTick
  (params (DeltaSeconds :type Float%20%28single-precision%29))
  (bind _returnvalue (Game|GetInstigatorController))
  ;[1 1 1] Tick - Update trajectory for local instigator
  (|CallParentFunction DeltaSeconds)
  (Utilities|IsValid (Variables|ParentBow|GetBowTrajectory)
    (:Is%20Valid
      (Utilities|IsValid _returnvalue
        (:Is%20Valid
          (if (and (Pawn|IsLocalPlayerController _returnvalue) (> (FortWeapon|GetChargePercent) 0.0))
            (CallFunction|IncreaseDrawStrength)))))))

(fn CalculateTrajectorySpline
  (bind (_startlocation _initialvelocity _projectilevelocity) (CallFunction|SetupProjectileVariables))
  (FortWeapon|UpdateDamageStartLocation)
  (bind (_outhitresult _outsplinepoints _outsplinetangents) (Projectile|CalculateProjectileTrajectorySplinewithHitProfile (Variables|ParentBow|GetPlayerPawn) _startlocation _initialvelocity _projectilevelocity (Variables|ParentBow|GetDummyGravity) (Variables|ParentBow|GetDummyFriction) (Variables|ParentBow|GetDummyBouncieness) (Variables|ParentBow|GetDummyTimestep) (Variables|ParentBow|GetDummyExtent) (Variables|ParentBow|GetTrajectoryCollisionProfile) (Class|FortPawn|GetProjectileIgnoredActors (Variables|ParentBow|GetPlayerPawn)) 1 20 (Math|Float|Lerp 3.5e+02 1e+03 (FortWeapon|GetChargePercent)) 1.5e+02))
  (Trajectory|SetTrajectorySpline (Variables|ParentBow|GetBowTrajectory) _outsplinepoints _outsplinetangents)
  (return _outhitresult))

(fn CalculateVelocity
  (return (FortWeapon|GetProjectileSpeed (FortWeapon|GetChargePercent))))

(fn GetChargedWeaponFireSound
  (params (Channel :type EFortWeaponSoundState%20Enum :default "Normal") (bSecondaryFire :type Boolean))
  (bind _returnvalue (Game|GetInstigator))
  (if (not bSecondaryFire)
    (Utilities|IsValid _returnvalue
      (:Is%20Valid
        (if (Pawn|IsLocallyControlled _returnvalue)
          (if (== (FortWeapon|GetChargePercent) 1.0)
            (return (Variables|ParentBow|GetFullTensionFireP1))
            (else
              (return (select (>= (FortWeapon|GetChargePercent) 0.5) (Variables|ParentBow|GetNormalTensionFireP1) (Variables|ParentBow|GetLowTensionFireP1)))))
          (else
            (if (== (FortWeapon|GetChargePercent) 1.0)
              (return (Variables|ParentBow|GetFullTensionFire))
              (else
                (return (select (>= (FortWeapon|GetChargePercent) 0.5) (Variables|ParentBow|GetNormalTensionFire) (Variables|ParentBow|GetLowTensionFire))))))))
      (:Is%20Not%20Valid
        (return 0)))
    (else
      (return 0))))

(fn GetFireSoundToPlay
  (if (Pawn|IsLocallyControlled (Game|GetInstigator))
    (return (select (Variables|ParentBow|GetReachedMaxCharge) (Variables|ParentBow|GetFullTensionFireP1) (select false (Variables|ParentBow|GetNormalTensionFireP1) (Variables|ParentBow|GetLowTensionFireP1))))
    (else
      (return (select (Variables|ParentBow|GetReachedMaxCharge) (Variables|ParentBow|GetFullTensionFire) (select false (Variables|ParentBow|GetNormalTensionFire) (Variables|ParentBow|GetLowTensionFire)))))))

(fn SetDummyVariables
  (bind _trajectorydata (Class|FortKismetLibrary|GetProjectileTrajectoryData (Variables|ParentBow|GetPlayerPawn) (Variables|ParentBow|GetPrj_Dummy)))
  (Variables|ParentBow|SetDummyMaxSpeed (Utilities|Struct|BreakFortProjectileTrajectoryData _trajectorydata))
  (Variables|ParentBow|SetDummyGravity (.GravityZ (Utilities|Struct|BreakFortProjectileTrajectoryData _trajectorydata)))
  (Variables|ParentBow|SetDummyExtent (Math|Float|Max(Float) (.CapsuleRadius (Utilities|Struct|BreakFortProjectileTrajectoryData _trajectorydata)) (.CapsuleHalfHeight (Utilities|Struct|BreakFortProjectileTrajectoryData _trajectorydata))))
  (Variables|ParentBow|SetDummyBouncieness (.Bounciness (Utilities|Struct|BreakFortProjectileTrajectoryData _trajectorydata)))
  (Variables|ParentBow|SetDummyFriction (.Friction (Utilities|Struct|BreakFortProjectileTrajectoryData _trajectorydata)))
  (Variables|ParentBow|SetDummyCollisionChannel (.CollisionObjectType (Utilities|Struct|BreakFortProjectileTrajectoryData _trajectorydata))))

(fn SetMaxChargeReachedAndPlayAudio
  (Variables|ParentBow|SetReachedMaxCharge true)
  (Audio|Components|Audio|Play (Variables|Default|GetFullCharge_AudioComponent) 0.0))

(fn SetupProjectileVariables
  (bind (_aimdir _returnvalue) (FortWeapon|GetDamageStartLocation -1))
  (return _returnvalue (* (Math|Rotator|GetRotationXVector (Math|Rotator|MakeRotator (.roll (Math|Conversions|RotationFromXVector _aimdir)) (+ (.pitch (Math|Conversions|RotationFromXVector _aimdir)) (FortWeapon|GetProjectilePitchOffset)) (.yaw (Math|Conversions|RotationFromXVector _aimdir)))) (Variables|ParentBow|GetProjectileVelocity)) (Variables|ParentBow|GetProjectileVelocity) (Math|Rotator|MakeRotator (.roll (Math|Conversions|RotationFromXVector _aimdir)) (+ (.pitch (Math|Conversions|RotationFromXVector _aimdir)) (FortWeapon|GetProjectilePitchOffset)) (.yaw (Math|Conversions|RotationFromXVector _aimdir)))))

(fn SetupTrajectoryProjectile
  ;[1 1 1] Generate the Trajectory indicator and save out the values for it.
  (bind _returnvalue (Game|SpawnActorfromClass (Variables|ParentBow|GetTrajectoryClass) (Transformation|GetActorTransform (Variables|ParentBow|GetPlayerPawn)) "AlwaysSpawn" "SelectDefaultAtRuntime" (Variables|ParentBow|GetPlayerPawn) (Variables|ParentBow|GetPlayerPawn)))
  (bind _asbp_projectile_trajectory_bow_generic (Utilities|Casting|CastToBP_ProjectileTrajectory_Bow_Generic _returnvalue))
  (Variables|/Game/Athena/Items/Weapons/Abilities/ExplosiveBow/BP_ProjectileTrajectory_Bow_Generic.BP_ProjectileTrajectory_Bow_Generic_C|SetOwnerWeapon self _asbp_projectile_trajectory_bow_generic)
  (bind _output_get (Variables|ParentBow|SetBowTrajectory _asbp_projectile_trajectory_bow_generic))
  (Actor|Tick|AddTickPrerequisiteActor _output_get self))

(fn SpawnFullyChargedFX
  (bind _returnvalue (Effects|Components|ParticleSystem|SpawnAltFXAttached "GenericAnimNotify" (Variables|ParentBow|GetNS_FullyCHarge) (Variables|ParentBow|GetFX_FullyCharge) (Variables|Visual|GetWeaponMesh) "Muzzle"))
  (Variables|Default|SetFX_FullyCharged _returnvalue))

(fn StartDrawAudio
  (Utilities|FlowControl|Sequence
    (:then_0
      (bind _charge_audiocomponent (Variables|Default|GetCharge_AudioComponent)
        (:then
          (Audio|Components|Audio|SetSound _charge_audiocomponent (Variables|ParentBow|GetSound_OnCharge))
          (Audio|Components|Audio|Play _charge_audiocomponent 0.0))
        (:else)))
    (:then_1
      (bind _sound_onfullycharged (Variables|ParentBow|GetSound_OnFullyCharged)
        (:then
          (Audio|Components|Audio|SetSound (Variables|Default|GetFullCharge_AudioComponent) _sound_onfullycharged))
        (:else)))))

(fn StopDrawAudio
  (Utilities|FlowControl|Sequence
    (:then_0
      (bind _charge_audiocomponent (Variables|Default|GetCharge_AudioComponent)
        (:then
          (Audio|Components|Audio|FadeOut _charge_audiocomponent 0.2 0.0 "Linear" :!class /Script/Engine.AudioComponent))
        (:else)))
    (:then_1
      (bind _fullcharge_audiocomponent (Variables|Default|GetFullCharge_AudioComponent)
        (:then
          (Audio|Components|Audio|FadeOut _fullcharge_audiocomponent 0.15 0.0 "Linear" :!class /Script/Engine.AudioComponent))
        (:else)))))

(fn ConstructionScript
  (|CallParentFunction))
